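# Animation Blueprint for the BattleHound Fierce pickaxe that adds secondary physics motion to the fur bones on its axe mesh.
# The cached Input_pose enters component space and passes through serial AnimDynamics nodes, each simulating one fur bone as a box with angular limits.
# The visible graph holds two such chains over the same bones: one simulates in world space, the other uses default simulation space with angular springs and zero-gravity override.
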AnimBlueprint /BRCosmetics/Weapons/FORT_Melee/Pickaxe_BattleHound_Fierce/Meshes/BattleHound_Fierce_Axe_AnimBP
target_skeleton=&/BattleHound_Fierce_Axe_Skeleton
graphs [
 AnimGraph {nodes [
   n1 = Anim Dynamics - Bone: dyn_fur_front_top_01!(ComponentPose=Local To Component(LocalPose=Use cached pose 'Input_pose'()), Alpha=1, GravityOverride:Vector="0,0,0", ExternalForce="0,0,0") -> (Pose) {
    angularBiasOverride=1
    angularDampingOverride=1
    angularSpringConstant=1000
    bGravityOverrideInSimSpace=true
    bOverrideAngularBias=true
    bOverrideAngularDamping=true
    bOverrideLinearDamping=true
    linearDampingOverride=1
    linearSpringConstant=1
    simulationSpace=World
    alphaBoolBlend {blendOption=Linear}
    alphaScaleBiasClamp {scale=0.3}
    boundBone {boneName=dyn_fur_front_top_01}
    componentAppliedLinearAccClamp=<1000 1000 1000>
    componentLinearAccScale=<1 1 1>
    componentLinearVelScale=<1 1 1>
    physicsBodyDefinitions [{
      boundBone {boneName=dyn_fur_front_top_01}
      boxExtents=<19.008 15.552 13.824>
      constraintSetup {angularTargetAxis=AxisZ, angularLimitsMax=<5 45 0>, angularLimitsMin=<-5 -45 -90>, angularTarget=<0 0 0>}
      localJointOffset=<1.9008 0 0>
     }]
   }
   n2 = Anim Dynamics - Bone: dyn_fur_front_bot_01!(ComponentPose=@n1, Alpha=1, GravityOverride:Vector="0,0,0", ExternalForce="0,0,0") -> (Pose) {
    angularBiasOverride=1
    angularDampingOverride=1
    angularSpringConstant=1000
    bGravityOverrideInSimSpace=true
    bOverrideAngularBias=true
    bOverrideAngularDamping=true
    bOverrideLinearDamping=true
    linearDampingOverride=1
    linearSpringConstant=1
    simulationSpace=World
    alphaBoolBlend {blendOption=Linear}
    alphaScaleBiasClamp {scale=0.3}
    boundBone {boneName=dyn_fur_front_bot_01}
    componentAppliedLinearAccClamp=<1000 1000 1000>
    componentLinearAccScale=<1 1 1>
    componentLinearVelScale=<1 1 1>
    physicsBodyDefinitions [{
      boundBone {boneName=dyn_fur_front_bot_01}
      boxExtents=<22.464 15.552 17.28>
      constraintSetup {angularTargetAxis=AxisZ, angularLimitsMax=<5 45 0>, angularLimitsMin=<-5 -45 -90>, angularTarget=<0 0 0>}
      localJointOffset=<1.5552 0 0>
     }]
   }
   n3 = Anim Dynamics - Bone: dyn_fur_back_top_01!(ComponentPose=@n2, Alpha=1, GravityOverride:Vector="0,0,0", ExternalForce="0,0,0") -> (Pose) {
    angularBiasOverride=1
    angularDampingOverride=1
    angularSpringConstant=1000
    bGravityOverrideInSimSpace=true
    bOverrideAngularBias=true
    bOverrideAngularDamping=true
    bOverrideLinearDamping=true
    linearDampingOverride=1
    linearSpringConstant=1
    simulationSpace=World
    alphaBoolBlend {blendOption=Linear}
    alphaScaleBiasClamp {scale=0.3}
    boundBone {boneName=dyn_fur_back_top_01}
    componentAppliedLinearAccClamp=<1000 1000 1000>
    componentLinearAccScale=<1 1 1>
    componentLinearVelScale=<1 1 1>
    physicsBodyDefinitions [{
      boundBone {boneName=dyn_fur_back_top_01}
      boxExtents=<17.28 12.096 12.096>
      constraintSetup {angularTargetAxis=AxisZ, angularLimitsMax=<5 45 0>, angularLimitsMin=<-5 -45 -90>, angularTarget=<0 0 0>}
      localJointOffset=<2.0736 0 0>
     }]
   }
   n4 = Anim Dynamics - Bone: dyn_fur_back_bot_01!(ComponentPose=@n3, Alpha=1, GravityOverride:Vector="0,0,0", ExternalForce="0,0,0") -> (Pose) {
    angularBiasOverride=1
    angularDampingOverride=1
    angularSpringConstant=1000
    bGravityOverrideInSimSpace=true
    bOverrideAngularBias=true
    bOverrideAngularDamping=true
    bOverrideLinearDamping=true
    linearDampingOverride=1
    linearSpringConstant=1
    simulationSpace=World
    alphaBoolBlend {blendOption=Linear}
    alphaScaleBiasClamp {scale=0.3}
    boundBone {boneName=dyn_fur_back_bot_01}
    componentAppliedLinearAccClamp=<1000 1000 1000>
    componentLinearAccScale=<1 1 1>
    componentLinearVelScale=<1 1 1>
    physicsBodyDefinitions [{
      boundBone {boneName=dyn_fur_back_bot_01}
      boxExtents=<22.464 12.096 17.28>
      constraintSetup {angularTargetAxis=AxisZ, angularLimitsMax=<5 45 0>, angularLimitsMin=<-5 -45 -90>, angularTarget=<0 0 0>}
      localJointOffset=<2.4192 0 0>
     }]
   }
   n5 = Anim Dynamics - Bone: dyn_fur_l_top_01!(ComponentPose=@n4, Alpha=1, GravityOverride:Vector="0,0,0", ExternalForce="0,0,0") -> (Pose) {
    angularBiasOverride=1
    angularDampingOverride=1
    angularSpringConstant=1000
    bGravityOverrideInSimSpace=true
    bOverrideAngularBias=true
    bOverrideAngularDamping=true
    bOverrideLinearDamping=true
    linearDampingOverride=1
    linearSpringConstant=1
    simulationSpace=World
    alphaBoolBlend {blendOption=Linear}
    alphaScaleBiasClamp {scale=0.3}
    boundBone {boneName=dyn_fur_l_top_01}
    componentAppliedLinearAccClamp=<1000 1000 1000>
    componentLinearAccScale=<1 1 1>
    componentLinearVelScale=<1 1 1>
    physicsBodyDefinitions [{
      boundBone {boneName=dyn_fur_l_top_01}
      boxExtents=<17.28 15.552 12.096>
      constraintSetup {angularTargetAxis=AxisZ, angularLimitsMax=<5 45 0>, angularLimitsMin=<-5 -45 -90>, angularTarget=<0 0 0>}
      localJointOffset=<1.3824 0 0>
     }]
   }
   n6 = Anim Dynamics - Bone: dyn_fur_l_bot_01!(ComponentPose=@n5, Alpha=1, GravityOverride:Vector="0,0,0", ExternalForce="0,0,0") -> (Pose) {
    angularBiasOverride=1
    angularDampingOverride=1
    angularSpringConstant=1000
    bGravityOverrideInSimSpace=true
    bOverrideAngularBias=true
    bOverrideAngularDamping=true
    bOverrideLinearDamping=true
    linearDampingOverride=1
    linearSpringConstant=1
    simulationSpace=World
    alphaBoolBlend {blendOption=Linear}
    alphaScaleBiasClamp {scale=0.3}
    boundBone {boneName=dyn_fur_l_bot_01}
    componentAppliedLinearAccClamp=<1000 1000 1000>
    componentLinearAccScale=<1 1 1>
    componentLinearVelScale=<1 1 1>
    physicsBodyDefinitions [{
      boundBone {boneName=dyn_fur_l_bot_01}
      boxExtents=<20.736 12.096 15.552>
      constraintSetup {angularTargetAxis=AxisZ, angularLimitsMax=<5 45 0>, angularLimitsMin=<-5 -45 -90>, angularTarget=<0 0 0>}
      localJointOffset=<1.728 0 0>
     }]
   }
   n7 = Anim Dynamics - Bone: dyn_fur_r_top_01!(ComponentPose=@n6, Alpha=1, GravityOverride:Vector="0,0,0", ExternalForce="0,0,0") -> (Pose) {
    angularBiasOverride=1
    angularDampingOverride=1
    angularSpringConstant=1000
    bGravityOverrideInSimSpace=true
    bOverrideAngularBias=true
    bOverrideAngularDamping=true
    bOverrideLinearDamping=true
    linearDampingOverride=1
    linearSpringConstant=1
    simulationSpace=World
    alphaBoolBlend {blendOption=Linear}
    alphaScaleBiasClamp {scale=0.3}
    boundBone {boneName=dyn_fur_r_top_01}
    componentAppliedLinearAccClamp=<1000 1000 1000>
    componentLinearAccScale=<1 1 1>
    componentLinearVelScale=<1 1 1>
    physicsBodyDefinitions [{
      boundBone {boneName=dyn_fur_r_top_01}
      boxExtents=<24.192 12.096 10.368>
      constraintSetup {angularTargetAxis=AxisZ, angularLimitsMax=<5 45 0>, angularLimitsMin=<-5 -45 -85>, angularTarget=<0 0 0>}
      localJointOffset=<1.5552 0 0>
     }]
   }
   n8 = Anim Dynamics - Bone: dyn_fur_r_bot_01!(ComponentPose=@n7, Alpha=1, GravityOverride:Vector="0,0,0", ExternalForce="0,0,0") -> (Pose) {
    angularBiasOverride=1
    angularDampingOverride=1
    angularSpringConstant=1000
    bGravityOverrideInSimSpace=true
    bOverrideAngularBias=true
    bOverrideAngularDamping=true
    bOverrideLinearDamping=true
    linearDampingOverride=1
    linearSpringConstant=1
    simulationSpace=World
    alphaBoolBlend {blendOption=Linear}
    alphaScaleBiasClamp {scale=0.3}
    boundBone {boneName=dyn_fur_r_bot_01}
    componentAppliedLinearAccClamp=<1000 1000 1000>
    componentLinearAccScale=<1 1 1>
    componentLinearVelScale=<1 1 1>
    physicsBodyDefinitions [{
      boundBone {boneName=dyn_fur_r_bot_01}
      boxExtents=<17.28 15.552 17.28>
      constraintSetup {angularTargetAxis=AxisZ, angularLimitsMax=<5 45 0>, angularLimitsMin=<-5 -45 -85>, angularTarget=<0 0 0>}
      localJointOffset=<2.2464 0 0>
     }]
   }
   n9 = Anim Dynamics - Bone: dyn_fur_front_top_A_01!(ComponentPose=@n8, Alpha=1, GravityOverride:Vector="0,0,0", ExternalForce="0,0,0") -> (Pose) {
    angularBiasOverride=1
    angularDampingOverride=1
    angularSpringConstant=1000
    bGravityOverrideInSimSpace=true
    bOverrideAngularBias=true
    bOverrideAngularDamping=true
    bOverrideLinearDamping=true
    linearDampingOverride=1
    linearSpringConstant=1
    simulationSpace=World
    alphaBoolBlend {blendOption=Linear}
    alphaScaleBiasClamp {scale=0.3}
    boundBone {boneName=dyn_fur_front_top_A_01}
    componentAppliedLinearAccClamp=<1000 1000 1000>
    componentLinearAccScale=<1 1 1>
    componentLinearVelScale=<1 1 1>
    physicsBodyDefinitions [{
      boundBone {boneName=dyn_fur_front_top_A_01}
      boxExtents=<13.2 10.8 9.6>
      constraintSetup {angularTargetAxis=AxisZ, angularLimitsMax=<5 45 0>, angularLimitsMin=<-5 -45 -90>, angularTarget=<0 0 0>}
      localJointOffset=<1.32 0 0>
     }]
   }
   n10 = Anim Dynamics - Bone: dyn_fur_front_top_B_01!(ComponentPose=@n9, Alpha=1, GravityOverride:Vector="0,0,0", ExternalForce="0,0,0") -> (Pose) {
    angularBiasOverride=1
    angularDampingOverride=1
    angularSpringConstant=1000
    bGravityOverrideInSimSpace=true
    bOverrideAngularBias=true
    bOverrideAngularDamping=true
    bOverrideLinearDamping=true
    linearDampingOverride=1
    linearSpringConstant=1
    simulationSpace=World
    alphaBoolBlend {blendOption=Linear}
    alphaScaleBiasClamp {scale=0.3}
    boundBone {boneName=dyn_fur_front_top_B_01}
    componentAppliedLinearAccClamp=<1000 1000 1000>
    componentLinearAccScale=<1 1 1>
    componentLinearVelScale=<1 1 1>
    physicsBodyDefinitions [{
      boundBone {boneName=dyn_fur_front_top_B_01}
      boxExtents=<13.2 10.8 9.6>
      constraintSetup {angularTargetAxis=AxisZ, angularLimitsMax=<5 45 0>, angularLimitsMin=<-5 -45 -90>, angularTarget=<0 0 0>}
      localJointOffset=<1.32 0 0>
     }]
   }
   n11 = Anim Dynamics - Bone: dyn_fur_front_top_C_01!(ComponentPose=@n10, Alpha=1, GravityOverride:Vector="0,0,0", ExternalForce="0,0,0") -> (Pose) {
    angularBiasOverride=1
    angularDampingOverride=1
    angularSpringConstant=1000
    bGravityOverrideInSimSpace=true
    bOverrideAngularBias=true
    bOverrideAngularDamping=true
    bOverrideLinearDamping=true
    linearDampingOverride=1
    linearSpringConstant=1
    simulationSpace=World
    alphaBoolBlend {blendOption=Linear}
    alphaScaleBiasClamp {scale=0.3}
    boundBone {boneName=dyn_fur_front_top_C_01}
    componentAppliedLinearAccClamp=<1000 1000 1000>
    componentLinearAccScale=<1 1 1>
    componentLinearVelScale=<1 1 1>
    physicsBodyDefinitions [{
      boundBone {boneName=dyn_fur_front_top_C_01}
      boxExtents=<13.2 10.8 9.6>
      constraintSetup {angularTargetAxis=AxisZ, angularLimitsMax=<5 45 0>, angularLimitsMin=<-5 -45 -90>, angularTarget=<0 0 0>}
      localJointOffset=<1.32 0 0>
     }]
   }
   n12 = Anim Dynamics - Bone: dyn_fur_front_top_D_01!(ComponentPose=@n11, Alpha=1, GravityOverride:Vector="0,0,0", ExternalForce="0,0,0") -> (Pose) {
    angularBiasOverride=1
    angularDampingOverride=1
    angularSpringConstant=1000
    bGravityOverrideInSimSpace=true
    bOverrideAngularBias=true
    bOverrideAngularDamping=true
    bOverrideLinearDamping=true
    linearDampingOverride=1
    linearSpringConstant=1
    simulationSpace=World
    alphaBoolBlend {blendOption=Linear}
    alphaScaleBiasClamp {scale=0.3}
    boundBone {boneName=dyn_fur_front_top_D_01}
    componentAppliedLinearAccClamp=<1000 1000 1000>
    componentLinearAccScale=<1 1 1>
    componentLinearVelScale=<1 1 1>
    physicsBodyDefinitions [{
      boundBone {boneName=dyn_fur_front_top_D_01}
      boxExtents=<13.2 10.8 9.6>
      constraintSetup {angularTargetAxis=AxisZ, angularLimitsMax=<5 45 0>, angularLimitsMin=<-5 -45 -90>, angularTarget=<0 0 0>}
      localJointOffset=<1.32 0 0>
     }]
   }
   n13 = Anim Dynamics - Bone: dyn_fur_front_bot_A_01!(ComponentPose=@n12, Alpha=1, GravityOverride:Vector="0,0,0", ExternalForce="0,0,0") -> (Pose) {
    angularBiasOverride=1
    angularDampingOverride=1
    angularSpringConstant=1000
    bGravityOverrideInSimSpace=true
    bOverrideAngularBias=true
    bOverrideAngularDamping=true
    bOverrideLinearDamping=true
    linearDampingOverride=1
    linearSpringConstant=1
    simulationSpace=World
    alphaBoolBlend {blendOption=Linear}
    alphaScaleBiasClamp {scale=0.3}
    boundBone {boneName=dyn_fur_front_bot_A_01}
    componentAppliedLinearAccClamp=<1000 1000 1000>
    componentLinearAccScale=<1 1 1>
    componentLinearVelScale=<1 1 1>
    physicsBodyDefinitions [{
      boundBone {boneName=dyn_fur_front_bot_A_01}
      boxExtents=<15.6 10.8 12>
      constraintSetup {angularTargetAxis=AxisZ, angularLimitsMax=<5 45 0>, angularLimitsMin=<-5 -45 -90>, angularTarget=<0 0 0>}
      localJointOffset=<1.08 0 0>
     }]
   }
   n14 = Anim Dynamics - Bone: dyn_fur_front_bot_B_01!(ComponentPose=@n13, Alpha=1, GravityOverride:Vector="0,0,0", ExternalForce="0,0,0") -> (Pose) {
    angularBiasOverride=1
    angularDampingOverride=1
    angularSpringConstant=1000
    bGravityOverrideInSimSpace=true
    bOverrideAngularBias=true
    bOverrideAngularDamping=true
    bOverrideLinearDamping=true
    linearDampingOverride=1
    linearSpringConstant=1
    simulationSpace=World
    alphaBoolBlend {blendOption=Linear}
    alphaScaleBiasClamp {scale=0.3}
    boundBone {boneName=dyn_fur_front_bot_B_01}
    componentAppliedLinearAccClamp=<1000 1000 1000>
    componentLinearAccScale=<1 1 1>
    componentLinearVelScale=<1 1 1>
    physicsBodyDefinitions [{
      boundBone {boneName=dyn_fur_front_bot_B_01}
      boxExtents=<15.6 10.8 12>
      constraintSetup {angularTargetAxis=AxisZ, angularLimitsMax=<5 45 0>, angularLimitsMin=<-5 -45 -90>, angularTarget=<0 0 0>}
      localJointOffset=<1.08 0 0>
     }]
   }
   n15 = Anim Dynamics - Bone: dyn_fur_back_top_A_01!(ComponentPose=@n14, Alpha=1, GravityOverride:Vector="0,0,0", ExternalForce="0,0,0") -> (Pose) {
    angularBiasOverride=1
    angularDampingOverride=1
    angularSpringConstant=1000
    bGravityOverrideInSimSpace=true
    bOverrideAngularBias=true
    bOverrideAngularDamping=true
    bOverrideLinearDamping=true
    linearDampingOverride=1
    linearSpringConstant=1
    simulationSpace=World
    alphaBoolBlend {blendOption=Linear}
    alphaScaleBiasClamp {scale=0.3}
    boundBone {boneName=dyn_fur_back_top_A_01}
    componentAppliedLinearAccClamp=<1000 1000 1000>
    componentLinearAccScale=<1 1 1>
    componentLinearVelScale=<1 1 1>
    physicsBodyDefinitions [{
      boundBone {boneName=dyn_fur_back_top_A_01}
      boxExtents=<12 8.4 8.4>
      constraintSetup {angularTargetAxis=AxisZ, angularLimitsMax=<5 45 0>, angularLimitsMin=<-5 -45 -90>, angularTarget=<0 0 0>}
      localJointOffset=<1.44 0 0>
     }]
   }
   n16 = Anim Dynamics - Bone: dyn_fur_back_top_B_01!(ComponentPose=@n15, Alpha=1, GravityOverride:Vector="0,0,0", ExternalForce="0,0,0") -> (Pose) {
    angularBiasOverride=1
    angularDampingOverride=1
    angularSpringConstant=1000
    bGravityOverrideInSimSpace=true
    bOverrideAngularBias=true
    bOverrideAngularDamping=true
    bOverrideLinearDamping=true
    linearDampingOverride=1
    linearSpringConstant=1
    simulationSpace=World
    alphaBoolBlend {blendOption=Linear}
    alphaScaleBiasClamp {scale=0.3}
    boundBone {boneName=dyn_fur_back_top_B_01}
    componentAppliedLinearAccClamp=<1000 1000 1000>
    componentLinearAccScale=<1 1 1>
    componentLinearVelScale=<1 1 1>
    physicsBodyDefinitions [{
      boundBone {boneName=dyn_fur_back_top_B_01}
      boxExtents=<12 8.4 8.4>
      constraintSetup {angularTargetAxis=AxisZ, angularLimitsMax=<5 45 0>, angularLimitsMin=<-5 -45 -90>, angularTarget=<0 0 0>}
      localJointOffset=<1.44 0 0>
     }]
   }
   n17 = Anim Dynamics - Bone: dyn_fur_back_top_C_01!(ComponentPose=@n16, Alpha=1, GravityOverride:Vector="0,0,0", ExternalForce="0,0,0") -> (Pose) {
    angularBiasOverride=1
    angularDampingOverride=1
    angularSpringConstant=1000
    bGravityOverrideInSimSpace=true
    bOverrideAngularBias=true
    bOverrideAngularDamping=true
    bOverrideLinearDamping=true
    linearDampingOverride=1
    linearSpringConstant=1
    simulationSpace=World
    alphaBoolBlend {blendOption=Linear}
    alphaScaleBiasClamp {scale=0.3}
    boundBone {boneName=dyn_fur_back_top_C_01}
    componentAppliedLinearAccClamp=<1000 1000 1000>
    componentLinearAccScale=<1 1 1>
    componentLinearVelScale=<1 1 1>
    physicsBodyDefinitions [{
      boundBone {boneName=dyn_fur_back_top_C_01}
      boxExtents=<12 8.4 8.4>
      constraintSetup {angularTargetAxis=AxisZ, angularLimitsMax=<5 45 0>, angularLimitsMin=<-5 -45 -90>, angularTarget=<0 0 0>}
      localJointOffset=<1.44 0 0>
     }]
   }
   n18 = Anim Dynamics - Bone: dyn_fur_back_bot_A_01!(ComponentPose=@n17, Alpha=1, GravityOverride:Vector="0,0,0", ExternalForce="0,0,0") -> (Pose) {
    angularBiasOverride=1
    angularDampingOverride=1
    angularSpringConstant=1000
    bGravityOverrideInSimSpace=true
    bOverrideAngularBias=true
    bOverrideAngularDamping=true
    bOverrideLinearDamping=true
    linearDampingOverride=1
    linearSpringConstant=1
    simulationSpace=World
    alphaBoolBlend {blendOption=Linear}
    alphaScaleBiasClamp {scale=0.3}
    boundBone {boneName=dyn_fur_back_bot_A_01}
    componentAppliedLinearAccClamp=<1000 1000 1000>
    componentLinearAccScale=<1 1 1>
    componentLinearVelScale=<1 1 1>
    physicsBodyDefinitions [{
      boundBone {boneName=dyn_fur_back_bot_A_01}
      boxExtents=<15.6 8.4 12>
      constraintSetup {angularTargetAxis=AxisZ, angularLimitsMax=<5 45 0>, angularLimitsMin=<-5 -45 -90>, angularTarget=<0 0 0>}
      localJointOffset=<1.68 0 0>
     }]
   }
   n19 = Anim Dynamics - Bone: dyn_fur_back_bot_B_01!(ComponentPose=@n18, Alpha=1, GravityOverride:Vector="0,0,0", ExternalForce="0,0,0") -> (Pose) {
    angularBiasOverride=1
    angularDampingOverride=1
    angularSpringConstant=1000
    bGravityOverrideInSimSpace=true
    bOverrideAngularBias=true
    bOverrideAngularDamping=true
    bOverrideLinearDamping=true
    linearDampingOverride=1
    linearSpringConstant=1
    simulationSpace=World
    alphaBoolBlend {blendOption=Linear}
    alphaScaleBiasClamp {scale=0.3}
    boundBone {boneName=dyn_fur_back_bot_B_01}
    componentAppliedLinearAccClamp=<1000 1000 1000>
    componentLinearAccScale=<1 1 1>
    componentLinearVelScale=<1 1 1>
    physicsBodyDefinitions [{
      boundBone {boneName=dyn_fur_back_bot_B_01}
      boxExtents=<15.6 8.4 12>
      constraintSetup {angularTargetAxis=AxisZ, angularLimitsMax=<5 45 0>, angularLimitsMin=<-5 -45 -90>, angularTarget=<0 0 0>}
      localJointOffset=<1.68 0 0>
     }]
   }
   n20 = Anim Dynamics - Bone: dyn_fur_back_bot_C_01!(ComponentPose=@n19, Alpha=1, GravityOverride:Vector="0,0,0", ExternalForce="0,0,0") -> (Pose) {
    angularBiasOverride=1
    angularDampingOverride=1
    angularSpringConstant=1000
    bGravityOverrideInSimSpace=true
    bOverrideAngularBias=true
    bOverrideAngularDamping=true
    bOverrideLinearDamping=true
    bUseGravityOverride=true
    linearDampingOverride=1
    linearSpringConstant=1
    simulationSpace=World
    alphaBoolBlend {blendOption=Linear}
    alphaScaleBiasClamp {scale=0.3}
    boundBone {boneName=dyn_fur_back_bot_C_01}
    componentAppliedLinearAccClamp=<1000 1000 1000>
    componentLinearAccScale=<1 1 1>
    componentLinearVelScale=<1 1 1>
    physicsBodyDefinitions [{
      boundBone {boneName=dyn_fur_back_bot_C_01}
      boxExtents=<15.6 8.4 12>
      constraintSetup {angularTargetAxis=AxisZ, angularLimitsMax=<5 45 0>, angularLimitsMin=<-5 -45 -90>, angularTarget=<0 0 0>}
      localJointOffset=<1.68 0 0>
     }]
   }
   n21 = Anim Dynamics - Bone: dyn_fur_l_top_A_01!(ComponentPose=@n20, Alpha=1, GravityOverride:Vector="0,0,0", ExternalForce="0,0,0") -> (Pose) {
    angularBiasOverride=1
    angularDampingOverride=1
    angularSpringConstant=1000
    bGravityOverrideInSimSpace=true
    bOverrideAngularBias=true
    bOverrideAngularDamping=true
    bOverrideLinearDamping=true
    linearDampingOverride=1
    linearSpringConstant=1
    simulationSpace=World
    alphaBoolBlend {blendOption=Linear}
    alphaScaleBiasClamp {scale=0.3}
    boundBone {boneName=dyn_fur_l_top_A_01}
    componentAppliedLinearAccClamp=<1000 1000 1000>
    componentLinearAccScale=<1 1 1>
    componentLinearVelScale=<1 1 1>
    physicsBodyDefinitions [{
      boundBone {boneName=dyn_fur_l_top_A_01}
      boxExtents=<12 10.8 8.4>
      constraintSetup {angularTargetAxis=AxisZ, angularLimitsMax=<5 45 0>, angularLimitsMin=<-5 -45 -90>, angularTarget=<0 0 0>}
      localJointOffset=<0.96 0 0>
     }]
   }
   n22 = Anim Dynamics - Bone: dyn_fur_l_top_B_01!(ComponentPose=@n21, Alpha=1, GravityOverride:Vector="0,0,0", ExternalForce="0,0,0") -> (Pose) {
    angularBiasOverride=1
    angularDampingOverride=1
    angularSpringConstant=1000
    bGravityOverrideInSimSpace=true
    bOverrideAngularBias=true
    bOverrideAngularDamping=true
    bOverrideLinearDamping=true
    linearDampingOverride=1
    linearSpringConstant=1
    simulationSpace=World
    alphaBoolBlend {blendOption=Linear}
    alphaScaleBiasClamp {scale=0.3}
    boundBone {boneName=dyn_fur_l_top_B_01}
    componentAppliedLinearAccClamp=<1000 1000 1000>
    componentLinearAccScale=<1 1 1>
    componentLinearVelScale=<1 1 1>
    physicsBodyDefinitions [{
      boundBone {boneName=dyn_fur_l_top_B_01}
      boxExtents=<12 10.8 8.4>
      constraintSetup {angularTargetAxis=AxisZ, angularLimitsMax=<5 45 0>, angularLimitsMin=<-5 -45 -90>, angularTarget=<0 0 0>}
      localJointOffset=<0.96 0 0>
     }]
   }
   n23 = Anim Dynamics - Bone: dyn_fur_l_bot_A_01!(ComponentPose=@n22, Alpha=1, GravityOverride:Vector="0,0,0", ExternalForce="0,0,0") -> (Pose) {
    angularBiasOverride=1
    angularDampingOverride=1
    angularSpringConstant=1000
    bGravityOverrideInSimSpace=true
    bOverrideAngularBias=true
    bOverrideAngularDamping=true
    bOverrideLinearDamping=true
    linearDampingOverride=1
    linearSpringConstant=1
    simulationSpace=World
    alphaBoolBlend {blendOption=Linear}
    alphaScaleBiasClamp {scale=0.3}
    boundBone {boneName=dyn_fur_l_bot_A_01}
    componentAppliedLinearAccClamp=<1000 1000 1000>
    componentLinearAccScale=<1 1 1>
    componentLinearVelScale=<1 1 1>
    physicsBodyDefinitions [{
      boundBone {boneName=dyn_fur_l_bot_A_01}
      boxExtents=<12 10.8 8.4>
      constraintSetup {angularTargetAxis=AxisZ, angularLimitsMax=<5 45 0>, angularLimitsMin=<-5 -45 -90>, angularTarget=<0 0 0>}
      localJointOffset=<0.96 0 0>
     }]
   }
   n24 = Anim Dynamics - Bone: dyn_fur_l_bot_B_01!(ComponentPose=@n23, Alpha=1, GravityOverride:Vector="0,0,0", ExternalForce="0,0,0") -> (Pose) {
    angularBiasOverride=1
    angularDampingOverride=1
    angularSpringConstant=1000
    bGravityOverrideInSimSpace=true
    bOverrideAngularBias=true
    bOverrideAngularDamping=true
    bOverrideLinearDamping=true
    linearDampingOverride=1
    linearSpringConstant=1
    simulationSpace=World
    alphaBoolBlend {blendOption=Linear}
    alphaScaleBiasClamp {scale=0.3}
    boundBone {boneName=dyn_fur_l_bot_B_01}
    componentAppliedLinearAccClamp=<1000 1000 1000>
    componentLinearAccScale=<1 1 1>
    componentLinearVelScale=<1 1 1>
    physicsBodyDefinitions [{
      boundBone {boneName=dyn_fur_l_bot_B_01}
      boxExtents=<12 10.8 8.4>
      constraintSetup {angularTargetAxis=AxisZ, angularLimitsMax=<5 45 0>, angularLimitsMin=<-5 -45 -90>, angularTarget=<0 0 0>}
      localJointOffset=<0.96 0 0>
     }]
   }
   n25 = Anim Dynamics - Bone: dyn_fur_r_top_A_01!(ComponentPose=@n24, Alpha=1, GravityOverride:Vector="0,0,0", ExternalForce="0,0,0") -> (Pose) {
    angularBiasOverride=1
    angularDampingOverride=1
    angularSpringConstant=1000
    bGravityOverrideInSimSpace=true
    bOverrideAngularBias=true
    bOverrideAngularDamping=true
    bOverrideLinearDamping=true
    linearDampingOverride=1
    linearSpringConstant=1
    simulationSpace=World
    alphaBoolBlend {blendOption=Linear}
    alphaScaleBiasClamp {scale=0.3}
    boundBone {boneName=dyn_fur_r_top_A_01}
    componentAppliedLinearAccClamp=<1000 1000 1000>
    componentLinearAccScale=<1 1 1>
    componentLinearVelScale=<1 1 1>
    physicsBodyDefinitions [{
      boundBone {boneName=dyn_fur_r_top_A_01}
      boxExtents=<16.8 8.4 7.2>
      constraintSetup {angularTargetAxis=AxisZ, angularLimitsMax=<5 45 0>, angularLimitsMin=<-5 -45 -85>, angularTarget=<0 0 0>}
      localJointOffset=<1.08 0 0>
     }]
   }
   n26 = Anim Dynamics - Bone: dyn_fur_r_top_B_01!(ComponentPose=@n25, Alpha=1, GravityOverride:Vector="0,0,0", ExternalForce="0,0,0") -> (Pose) {
    angularBiasOverride=1
    angularDampingOverride=1
    angularSpringConstant=1000
    bGravityOverrideInSimSpace=true
    bOverrideAngularBias=true
    bOverrideAngularDamping=true
    bOverrideLinearDamping=true
    linearDampingOverride=1
    linearSpringConstant=1
    simulationSpace=World
    alphaBoolBlend {blendOption=Linear}
    alphaScaleBiasClamp {scale=0.3}
    boundBone {boneName=dyn_fur_r_top_B_01}
    componentAppliedLinearAccClamp=<1000 1000 1000>
    componentLinearAccScale=<1 1 1>
    componentLinearVelScale=<1 1 1>
    physicsBodyDefinitions [{
      boundBone {boneName=dyn_fur_r_top_B_01}
      boxExtents=<16.8 8.4 7.2>
      constraintSetup {angularTargetAxis=AxisZ, angularLimitsMax=<5 45 0>, angularLimitsMin=<-5 -45 -85>, angularTarget=<0 0 0>}
      localJointOffset=<1.08 0 0>
     }]
   }
   n27 = Anim Dynamics - Bone: dyn_fur_r_top_C_01!(ComponentPose=@n26, Alpha=1, GravityOverride:Vector="0,0,0", ExternalForce="0,0,0") -> (Pose) {
    angularBiasOverride=1
    angularDampingOverride=1
    angularSpringConstant=1000
    bGravityOverrideInSimSpace=true
    bOverrideAngularBias=true
    bOverrideAngularDamping=true
    bOverrideLinearDamping=true
    linearDampingOverride=1
    linearSpringConstant=1
    simulationSpace=World
    alphaBoolBlend {blendOption=Linear}
    alphaScaleBiasClamp {scale=0.3}
    boundBone {boneName=dyn_fur_r_top_C_01}
    componentAppliedLinearAccClamp=<1000 1000 1000>
    componentLinearAccScale=<1 1 1>
    componentLinearVelScale=<1 1 1>
    physicsBodyDefinitions [{
      boundBone {boneName=dyn_fur_r_top_C_01}
      boxExtents=<16.8 8.4 7.2>
      constraintSetup {angularTargetAxis=AxisZ, angularLimitsMax=<5 45 0>, angularLimitsMin=<-5 -45 -85>, angularTarget=<0 0 0>}
      localJointOffset=<1.08 0 0>
     }]
   }
   n28 = Anim Dynamics - Bone: dyn_fur_r_bot_A_01!(ComponentPose=@n27, Alpha=1, GravityOverride:Vector="0,0,0", ExternalForce="0,0,0") -> (Pose) {
    angularBiasOverride=1
    angularDampingOverride=1
    angularSpringConstant=1000
    bGravityOverrideInSimSpace=true
    bOverrideAngularBias=true
    bOverrideAngularDamping=true
    bOverrideLinearDamping=true
    linearDampingOverride=1
    linearSpringConstant=1
    simulationSpace=World
    alphaBoolBlend {blendOption=Linear}
    alphaScaleBiasClamp {scale=0.3}
    boundBone {boneName=dyn_fur_r_bot_A_01}
    componentAppliedLinearAccClamp=<1000 1000 1000>
    componentLinearAccScale=<1 1 1>
    componentLinearVelScale=<1 1 1>
    physicsBodyDefinitions [{
      boundBone {boneName=dyn_fur_r_bot_A_01}
      boxExtents=<16.8 8.4 7.2>
      constraintSetup {angularTargetAxis=AxisZ, angularLimitsMax=<5 45 0>, angularLimitsMin=<-5 -45 -85>, angularTarget=<0 0 0>}
      localJointOffset=<1.08 0 0>
     }]
   }
   n29 = Anim Dynamics - Bone: dyn_fur_r_bot_B_01!(ComponentPose=@n28, Alpha=1, GravityOverride:Vector="0,0,0", ExternalForce="0,0,0") -> (Pose) {
    angularBiasOverride=1
    angularDampingOverride=1
    angularSpringConstant=1000
    bGravityOverrideInSimSpace=true
    bOverrideAngularBias=true
    bOverrideAngularDamping=true
    bOverrideLinearDamping=true
    linearDampingOverride=1
    linearSpringConstant=1
    simulationSpace=World
    alphaBoolBlend {blendOption=Linear}
    alphaScaleBiasClamp {scale=0.3}
    boundBone {boneName=dyn_fur_r_bot_B_01}
    componentAppliedLinearAccClamp=<1000 1000 1000>
    componentLinearAccScale=<1 1 1>
    componentLinearVelScale=<1 1 1>
    physicsBodyDefinitions [{
      boundBone {boneName=dyn_fur_r_bot_B_01}
      boxExtents=<16.8 8.4 7.2>
      constraintSetup {angularTargetAxis=AxisZ, angularLimitsMax=<5 45 0>, angularLimitsMin=<-5 -45 -85>, angularTarget=<0 0 0>}
      localJointOffset=<1.08 0 0>
     }]
   }
   n30 = Anim Dynamics - Bone: dyn_fur_front_top_01(ComponentPose=Local To Component(LocalPose=Use cached pose 'Input_pose'()), Alpha=1, GravityOverride="0,0,0", ExternalForce="0,0,0") {
    angularBiasOverride=1
    angularDampingOverride=1
    angularSpringConstant=1000
    bAngularSpring=true
    bGravityOverrideInSimSpace=true
    bOverrideAngularBias=true
    bOverrideAngularDamping=true
    bOverrideLinearDamping=true
    bUseGravityOverride=true
    linearDampingOverride=1
    linearSpringConstant=1
    alphaBoolBlend {blendOption=Linear}
    alphaScaleBiasClamp {scale=0.4}
    boundBone {boneName=dyn_fur_front_top_01}
    componentAppliedLinearAccClamp=<1000 1000 1000>
    componentLinearAccScale=<1 1 1>
    componentLinearVelScale=<1 1 1>
    physicsBodyDefinitions [{
      boundBone {boneName=dyn_fur_front_top_01}
      boxExtents=<19.008 15.552 13.824>
      constraintSetup {angularTargetAxis=AxisZ, angularLimitsMax=<5 45 0>, angularLimitsMin=<-5 -45 -90>, angularTarget=<0 0 0>}
      localJointOffset=<1.9008 0 0>
     }]
   }
   n31 = Anim Dynamics - Bone: dyn_fur_front_bot_01(ComponentPose=@n30, Alpha=1, GravityOverride="0,0,0", ExternalForce="0,0,0") {
    angularBiasOverride=1
    angularDampingOverride=1
    angularSpringConstant=1000
    bAngularSpring=true
    bGravityOverrideInSimSpace=true
    bOverrideAngularBias=true
    bOverrideAngularDamping=true
    bOverrideLinearDamping=true
    bUseGravityOverride=true
    linearDampingOverride=1
    linearSpringConstant=1
    alphaBoolBlend {blendOption=Linear}
    alphaScaleBiasClamp {scale=0.4}
    boundBone {boneName=dyn_fur_front_bot_01}
    componentAppliedLinearAccClamp=<1000 1000 1000>
    componentLinearAccScale=<1 1 1>
    componentLinearVelScale=<1 1 1>
    physicsBodyDefinitions [{
      boundBone {boneName=dyn_fur_front_bot_01}
      boxExtents=<22.464 15.552 17.28>
      constraintSetup {angularTargetAxis=AxisZ, angularLimitsMax=<5 45 0>, angularLimitsMin=<-5 -45 -90>, angularTarget=<0 0 0>}
      localJointOffset=<1.5552 0 0>
     }]
   }
   n32 = Anim Dynamics - Bone: dyn_fur_back_top_01(ComponentPose=@n31, Alpha=1, GravityOverride="0,0,0", ExternalForce="0,0,0") {
    angularBiasOverride=1
    angularDampingOverride=1
    angularSpringConstant=1000
    bAngularSpring=true
    bGravityOverrideInSimSpace=true
    bOverrideAngularBias=true
    bOverrideAngularDamping=true
    bOverrideLinearDamping=true
    bUseGravityOverride=true
    linearDampingOverride=1
    linearSpringConstant=1
    alphaBoolBlend {blendOption=Linear}
    alphaScaleBiasClamp {scale=0.4}
    boundBone {boneName=dyn_fur_back_top_01}
    componentAppliedLinearAccClamp=<1000 1000 1000>
    componentLinearAccScale=<1 1 1>
    componentLinearVelScale=<1 1 1>
    physicsBodyDefinitions [{
      boundBone {boneName=dyn_fur_back_top_01}
      boxExtents=<17.28 12.096 12.096>
      constraintSetup {angularTargetAxis=AxisZ, angularLimitsMax=<5 45 0>, angularLimitsMin=<-5 -45 -90>, angularTarget=<0 0 0>}
      localJointOffset=<2.0736 0 0>
     }]
   }
   n33 = Anim Dynamics - Bone: dyn_fur_back_bot_01(ComponentPose=@n32, Alpha=1, GravityOverride="0,0,0", ExternalForce="0,0,0") {
    angularBiasOverride=1
    angularDampingOverride=1
    angularSpringConstant=1000
    bAngularSpring=true
    bGravityOverrideInSimSpace=true
    bOverrideAngularBias=true
    bOverrideAngularDamping=true
    bOverrideLinearDamping=true
    bUseGravityOverride=true
    linearDampingOverride=1
    linearSpringConstant=1
    alphaBoolBlend {blendOption=Linear}
    alphaScaleBiasClamp {scale=0.4}
    boundBone {boneName=dyn_fur_back_bot_01}
    componentAppliedLinearAccClamp=<1000 1000 1000>
    componentLinearAccScale=<1 1 1>
    componentLinearVelScale=<1 1 1>
    physicsBodyDefinitions [{
      boundBone {boneName=dyn_fur_back_bot_01}
      boxExtents=<22.464 12.096 17.28>
      constraintSetup {angularTargetAxis=AxisZ, angularLimitsMax=<5 45 0>, angularLimitsMin=<-5 -45 -90>, angularTarget=<0 0 0>}
      localJointOffset=<2.4192 0 0>
     }]
   }
   n34 = Anim Dynamics - Bone: dyn_fur_l_top_01(ComponentPose=@n33, Alpha=1, GravityOverride="0,0,0", ExternalForce="0,0,0") {
    angularBiasOverride=1
    angularDampingOverride=1
    angularSpringConstant=1000
    bAngularSpring=true
    bGravityOverrideInSimSpace=true
    bOverrideAngularBias=true
    bOverrideAngularDamping=true
    bOverrideLinearDamping=true
    bUseGravityOverride=true
    linearDampingOverride=1
    linearSpringConstant=1
    alphaBoolBlend {blendOption=Linear}
    alphaScaleBiasClamp {scale=0.4}
    boundBone {boneName=dyn_fur_l_top_01}
    componentAppliedLinearAccClamp=<1000 1000 1000>
    componentLinearAccScale=<1 1 1>
    componentLinearVelScale=<1 1 1>
    physicsBodyDefinitions [{
      boundBone {boneName=dyn_fur_l_top_01}
      boxExtents=<17.28 15.552 12.096>
      constraintSetup {angularTargetAxis=AxisZ, angularLimitsMax=<5 45 0>, angularLimitsMin=<-5 -45 -90>, angularTarget=<0 0 0>}
      localJointOffset=<1.3824 0 0>
     }]
   }
   n35 = Anim Dynamics - Bone: dyn_fur_l_bot_01(ComponentPose=@n34, Alpha=1, GravityOverride="0,0,0", ExternalForce="0,0,0") {
    angularBiasOverride=1
    angularDampingOverride=1
    angularSpringConstant=1000
    bAngularSpring=true
    bGravityOverrideInSimSpace=true
    bOverrideAngularBias=true
    bOverrideAngularDamping=true
    bOverrideLinearDamping=true
    bUseGravityOverride=true
    linearDampingOverride=1
    linearSpringConstant=1
    alphaBoolBlend {blendOption=Linear}
    alphaScaleBiasClamp {scale=0.4}
    boundBone {boneName=dyn_fur_l_bot_01}
    componentAppliedLinearAccClamp=<1000 1000 1000>
    componentLinearAccScale=<1 1 1>
    componentLinearVelScale=<1 1 1>
    physicsBodyDefinitions [{
      boundBone {boneName=dyn_fur_l_bot_01}
      boxExtents=<20.736 12.096 15.552>
      constraintSetup {angularTargetAxis=AxisZ, angularLimitsMax=<5 45 0>, angularLimitsMin=<-5 -45 -90>, angularTarget=<0 0 0>}
      localJointOffset=<1.728 0 0>
     }]
   }
   n36 = Anim Dynamics - Bone: dyn_fur_r_top_01(ComponentPose=@n35, Alpha=1, GravityOverride="0,0,0", ExternalForce="0,0,0") {
    angularBiasOverride=1
    angularDampingOverride=1
    angularSpringConstant=1000
    bAngularSpring=true
    bGravityOverrideInSimSpace=true
    bOverrideAngularBias=true
    bOverrideAngularDamping=true
    bOverrideLinearDamping=true
    bUseGravityOverride=true
    linearDampingOverride=1
    linearSpringConstant=1
    alphaBoolBlend {blendOption=Linear}
    alphaScaleBiasClamp {scale=0.4}
    boundBone {boneName=dyn_fur_r_top_01}
    componentAppliedLinearAccClamp=<1000 1000 1000>
    componentLinearAccScale=<1 1 1>
    componentLinearVelScale=<1 1 1>
    physicsBodyDefinitions [{
      boundBone {boneName=dyn_fur_r_top_01}
      boxExtents=<24.192 12.096 10.368>
      constraintSetup {angularTargetAxis=AxisZ, angularLimitsMax=<5 45 0>, angularLimitsMin=<-5 -45 -85>, angularTarget=<0 0 0>}
      localJointOffset=<1.5552 0 0>
     }]
   }
   n37 = Anim Dynamics - Bone: dyn_fur_r_bot_01(ComponentPose=@n36, Alpha=1, GravityOverride="0,0,0", ExternalForce="0,0,0") {
    angularBiasOverride=1
    angularDampingOverride=1
    angularSpringConstant=1000
    bAngularSpring=true
    bGravityOverrideInSimSpace=true
    bOverrideAngularBias=true
    bOverrideAngularDamping=true
    bOverrideLinearDamping=true
    bUseGravityOverride=true
    linearDampingOverride=1
    linearSpringConstant=1
    alphaBoolBlend {blendOption=Linear}
    alphaScaleBiasClamp {scale=0.4}
    boundBone {boneName=dyn_fur_r_bot_01}
    componentAppliedLinearAccClamp=<1000 1000 1000>
    componentLinearAccScale=<1 1 1>
    componentLinearVelScale=<1 1 1>
    physicsBodyDefinitions [{
      boundBone {boneName=dyn_fur_r_bot_01}
      boxExtents=<17.28 15.552 17.28>
      constraintSetup {angularTargetAxis=AxisZ, angularLimitsMax=<5 45 0>, angularLimitsMin=<-5 -45 -85>, angularTarget=<0 0 0>}
      localJointOffset=<2.2464 0 0>
     }]
   }
   n38 = Anim Dynamics - Bone: dyn_fur_front_top_A_01(ComponentPose=@n37, Alpha=1, GravityOverride="0,0,0", ExternalForce="0,0,0") {
    angularBiasOverride=1
    angularDampingOverride=1
    angularSpringConstant=1000
    bAngularSpring=true
    bGravityOverrideInSimSpace=true
    bOverrideAngularBias=true
    bOverrideAngularDamping=true
    bOverrideLinearDamping=true
    bUseGravityOverride=true
    linearDampingOverride=1
    linearSpringConstant=1
    alphaBoolBlend {blendOption=Linear}
    alphaScaleBiasClamp {scale=0.4}
    boundBone {boneName=dyn_fur_front_top_A_01}
    componentAppliedLinearAccClamp=<1000 1000 1000>
    componentLinearAccScale=<1 1 1>
    componentLinearVelScale=<1 1 1>
    physicsBodyDefinitions [{
      boundBone {boneName=dyn_fur_front_top_A_01}
      boxExtents=<13.2 10.8 9.6>
      constraintSetup {angularTargetAxis=AxisZ, angularLimitsMax=<5 45 0>, angularLimitsMin=<-5 -45 -90>, angularTarget=<0 0 0>}
      localJointOffset=<1.32 0 0>
     }]
   }
   n39 = Anim Dynamics - Bone: dyn_fur_front_top_B_01(ComponentPose=@n38, Alpha=1, GravityOverride="0,0,0", ExternalForce="0,0,0") {
    angularBiasOverride=1
    angularDampingOverride=1
    angularSpringConstant=1000
    bAngularSpring=true
    bGravityOverrideInSimSpace=true
    bOverrideAngularBias=true
    bOverrideAngularDamping=true
    bOverrideLinearDamping=true
    bUseGravityOverride=true
    linearDampingOverride=1
    linearSpringConstant=1
    alphaBoolBlend {blendOption=Linear}
    alphaScaleBiasClamp {scale=0.4}
    boundBone {boneName=dyn_fur_front_top_B_01}
    componentAppliedLinearAccClamp=<1000 1000 1000>
    componentLinearAccScale=<1 1 1>
    componentLinearVelScale=<1 1 1>
    physicsBodyDefinitions [{
      boundBone {boneName=dyn_fur_front_top_B_01}
      boxExtents=<13.2 10.8 9.6>
      constraintSetup {angularTargetAxis=AxisZ, angularLimitsMax=<5 45 0>, angularLimitsMin=<-5 -45 -90>, angularTarget=<0 0 0>}
      localJointOffset=<1.32 0 0>
     }]
   }
   n40 = Anim Dynamics - Bone: dyn_fur_front_top_C_01(ComponentPose=@n39, Alpha=1, GravityOverride="0,0,0", ExternalForce="0,0,0") {
    angularBiasOverride=1
    angularDampingOverride=1
    angularSpringConstant=1000
    bAngularSpring=true
    bGravityOverrideInSimSpace=true
    bOverrideAngularBias=true
    bOverrideAngularDamping=true
    bOverrideLinearDamping=true
    bUseGravityOverride=true
    linearDampingOverride=1
    linearSpringConstant=1
    alphaBoolBlend {blendOption=Linear}
    alphaScaleBiasClamp {scale=0.4}
    boundBone {boneName=dyn_fur_front_top_C_01}
    componentAppliedLinearAccClamp=<1000 1000 1000>
    componentLinearAccScale=<1 1 1>
    componentLinearVelScale=<1 1 1>
    physicsBodyDefinitions [{
      boundBone {boneName=dyn_fur_front_top_C_01}
      boxExtents=<13.2 10.8 9.6>
      constraintSetup {angularTargetAxis=AxisZ, angularLimitsMax=<5 45 0>, angularLimitsMin=<-5 -45 -90>, angularTarget=<0 0 0>}
      localJointOffset=<1.32 0 0>
     }]
   }
   n41 = Anim Dynamics - Bone: dyn_fur_front_top_D_01(ComponentPose=@n40, Alpha=1, GravityOverride="0,0,0", ExternalForce="0,0,0") {
    angularBiasOverride=1
    angularDampingOverride=1
    angularSpringConstant=1000
    bAngularSpring=true
    bGravityOverrideInSimSpace=true
    bOverrideAngularBias=true
    bOverrideAngularDamping=true
    bOverrideLinearDamping=true
    bUseGravityOverride=true
    linearDampingOverride=1
    linearSpringConstant=1
    alphaBoolBlend {blendOption=Linear}
    alphaScaleBiasClamp {scale=0.4}
    boundBone {boneName=dyn_fur_front_top_D_01}
    componentAppliedLinearAccClamp=<1000 1000 1000>
    componentLinearAccScale=<1 1 1>
    componentLinearVelScale=<1 1 1>
    physicsBodyDefinitions [{
      boundBone {boneName=dyn_fur_front_top_D_01}
      boxExtents=<13.2 10.8 9.6>
      constraintSetup {angularTargetAxis=AxisZ, angularLimitsMax=<5 45 0>, angularLimitsMin=<-5 -45 -90>, angularTarget=<0 0 0>}
      localJointOffset=<1.32 0 0>
     }]
   }
   n42 = Anim Dynamics - Bone: dyn_fur_front_bot_A_01(ComponentPose=@n41, Alpha=1, GravityOverride="0,0,0", ExternalForce="0,0,0") {
    angularBiasOverride=1
    angularDampingOverride=1
    angularSpringConstant=1000
    bAngularSpring=true
    bGravityOverrideInSimSpace=true
    bOverrideAngularBias=true
    bOverrideAngularDamping=true
    bOverrideLinearDamping=true
    bUseGravityOverride=true
    linearDampingOverride=1
    linearSpringConstant=1
    alphaBoolBlend {blendOption=Linear}
    alphaScaleBiasClamp {scale=0.4}
    boundBone {boneName=dyn_fur_front_bot_A_01}
    componentAppliedLinearAccClamp=<1000 1000 1000>
    componentLinearAccScale=<1 1 1>
    componentLinearVelScale=<1 1 1>
    physicsBodyDefinitions [{
      boundBone {boneName=dyn_fur_front_bot_A_01}
      boxExtents=<15.6 10.8 12>
      constraintSetup {angularTargetAxis=AxisZ, angularLimitsMax=<5 45 0>, angularLimitsMin=<-5 -45 -90>, angularTarget=<0 0 0>}
      localJointOffset=<1.08 0 0>
     }]
   }
   n43 = Anim Dynamics - Bone: dyn_fur_front_bot_B_01(ComponentPose=@n42, Alpha=1, GravityOverride="0,0,0", ExternalForce="0,0,0") {
    angularBiasOverride=1
    angularDampingOverride=1
    angularSpringConstant=1000
    bAngularSpring=true
    bGravityOverrideInSimSpace=true
    bOverrideAngularBias=true
    bOverrideAngularDamping=true
    bOverrideLinearDamping=true
    bUseGravityOverride=true
    linearDampingOverride=1
    linearSpringConstant=1
    alphaBoolBlend {blendOption=Linear}
    alphaScaleBiasClamp {scale=0.4}
    boundBone {boneName=dyn_fur_front_bot_B_01}
    componentAppliedLinearAccClamp=<1000 1000 1000>
    componentLinearAccScale=<1 1 1>
    componentLinearVelScale=<1 1 1>
    physicsBodyDefinitions [{
      boundBone {boneName=dyn_fur_front_bot_B_01}
      boxExtents=<15.6 10.8 12>
      constraintSetup {angularTargetAxis=AxisZ, angularLimitsMax=<5 45 0>, angularLimitsMin=<-5 -45 -90>, angularTarget=<0 0 0>}
      localJointOffset=<1.08 0 0>
     }]
   }
   n44 = Anim Dynamics - Bone: dyn_fur_back_top_A_01(ComponentPose=@n43, Alpha=1, GravityOverride="0,0,0", ExternalForce="0,0,0") {
    angularBiasOverride=1
    angularDampingOverride=1
    angularSpringConstant=1000
    bAngularSpring=true
    bGravityOverrideInSimSpace=true
    bOverrideAngularBias=true
    bOverrideAngularDamping=true
    bOverrideLinearDamping=true
    bUseGravityOverride=true
    linearDampingOverride=1
    linearSpringConstant=1
    alphaBoolBlend {blendOption=Linear}
    alphaScaleBiasClamp {scale=0.4}
    boundBone {boneName=dyn_fur_back_top_A_01}
    componentAppliedLinearAccClamp=<1000 1000 1000>
    componentLinearAccScale=<1 1 1>
    componentLinearVelScale=<1 1 1>
    physicsBodyDefinitions [{
      boundBone {boneName=dyn_fur_back_top_A_01}
      boxExtents=<12 8.4 8.4>
      constraintSetup {angularTargetAxis=AxisZ, angularLimitsMax=<5 45 0>, angularLimitsMin=<-5 -45 -90>, angularTarget=<0 0 0>}
      localJointOffset=<1.44 0 0>
     }]
   }
   n45 = Anim Dynamics - Bone: dyn_fur_back_top_B_01(ComponentPose=@n44, Alpha=1, GravityOverride="0,0,0", ExternalForce="0,0,0") {
    angularBiasOverride=1
    angularDampingOverride=1
    angularSpringConstant=1000
    bAngularSpring=true
    bGravityOverrideInSimSpace=true
    bOverrideAngularBias=true
    bOverrideAngularDamping=true
    bOverrideLinearDamping=true
    bUseGravityOverride=true
    linearDampingOverride=1
    linearSpringConstant=1
    alphaBoolBlend {blendOption=Linear}
    alphaScaleBiasClamp {scale=0.4}
    boundBone {boneName=dyn_fur_back_top_B_01}
    componentAppliedLinearAccClamp=<1000 1000 1000>
    componentLinearAccScale=<1 1 1>
    componentLinearVelScale=<1 1 1>
    physicsBodyDefinitions [{
      boundBone {boneName=dyn_fur_back_top_B_01}
      boxExtents=<12 8.4 8.4>
      constraintSetup {angularTargetAxis=AxisZ, angularLimitsMax=<5 45 0>, angularLimitsMin=<-5 -45 -90>, angularTarget=<0 0 0>}
      localJointOffset=<1.44 0 0>
     }]
   }
   n46 = Anim Dynamics - Bone: dyn_fur_back_top_C_01(ComponentPose=@n45, Alpha=1, GravityOverride="0,0,0", ExternalForce="0,0,0") {
    angularBiasOverride=1
    angularDampingOverride=1
    angularSpringConstant=1000
    bAngularSpring=true
    bGravityOverrideInSimSpace=true
    bOverrideAngularBias=true
    bOverrideAngularDamping=true
    bOverrideLinearDamping=true
    bUseGravityOverride=true
    linearDampingOverride=1
    linearSpringConstant=1
    alphaBoolBlend {blendOption=Linear}
    alphaScaleBiasClamp {scale=0.4}
    boundBone {boneName=dyn_fur_back_top_C_01}
    componentAppliedLinearAccClamp=<1000 1000 1000>
    componentLinearAccScale=<1 1 1>
    componentLinearVelScale=<1 1 1>
    physicsBodyDefinitions [{
      boundBone {boneName=dyn_fur_back_top_C_01}
      boxExtents=<12 8.4 8.4>
      constraintSetup {angularTargetAxis=AxisZ, angularLimitsMax=<5 45 0>, angularLimitsMin=<-5 -45 -90>, angularTarget=<0 0 0>}
      localJointOffset=<1.44 0 0>
     }]
   }
   n47 = Anim Dynamics - Bone: dyn_fur_back_bot_A_01(ComponentPose=@n46, Alpha=1, GravityOverride="0,0,0", ExternalForce="0,0,0") {
    angularBiasOverride=1
    angularDampingOverride=1
    angularSpringConstant=1000
    bAngularSpring=true
    bGravityOverrideInSimSpace=true
    bOverrideAngularBias=true
    bOverrideAngularDamping=true
    bOverrideLinearDamping=true
    bUseGravityOverride=true
    linearDampingOverride=1
    linearSpringConstant=1
    alphaBoolBlend {blendOption=Linear}
    alphaScaleBiasClamp {scale=0.4}
    boundBone {boneName=dyn_fur_back_bot_A_01}
    componentAppliedLinearAccClamp=<1000 1000 1000>
    componentLinearAccScale=<1 1 1>
    componentLinearVelScale=<1 1 1>
    physicsBodyDefinitions [{
      boundBone {boneName=dyn_fur_back_bot_A_01}
      boxExtents=<15.6 8.4 12>
      constraintSetup {angularTargetAxis=AxisZ, angularLimitsMax=<5 45 0>, angularLimitsMin=<-5 -45 -90>, angularTarget=<0 0 0>}
      localJointOffset=<1.68 0 0>
     }]
   }
   n48 = Anim Dynamics - Bone: dyn_fur_back_bot_B_01(ComponentPose=@n47, Alpha=1, GravityOverride="0,0,0", ExternalForce="0,0,0") {
    angularBiasOverride=1
    angularDampingOverride=1
    angularSpringConstant=1000
    bAngularSpring=true
    bGravityOverrideInSimSpace=true
    bOverrideAngularBias=true
    bOverrideAngularDamping=true
    bOverrideLinearDamping=true
    bUseGravityOverride=true
    linearDampingOverride=1
    linearSpringConstant=1
    alphaBoolBlend {blendOption=Linear}
    alphaScaleBiasClamp {scale=0.4}
    boundBone {boneName=dyn_fur_back_bot_B_01}
    componentAppliedLinearAccClamp=<1000 1000 1000>
    componentLinearAccScale=<1 1 1>
    componentLinearVelScale=<1 1 1>
    physicsBodyDefinitions [{
      boundBone {boneName=dyn_fur_back_bot_B_01}
      boxExtents=<15.6 8.4 12>
      constraintSetup {angularTargetAxis=AxisZ, angularLimitsMax=<5 45 0>, angularLimitsMin=<-5 -45 -90>, angularTarget=<0 0 0>}
      localJointOffset=<1.68 0 0>
     }]
   }
   n49 = Anim Dynamics - Bone: dyn_fur_back_bot_C_01(ComponentPose=@n48, Alpha=1, GravityOverride="0,0,0", ExternalForce="0,0,0") {
    angularBiasOverride=1
    angularDampingOverride=1
    angularSpringConstant=1000
    bAngularSpring=true
    bGravityOverrideInSimSpace=true
    bOverrideAngularBias=true
    bOverrideAngularDamping=true
    bOverrideLinearDamping=true
    bUseGravityOverride=true
    linearDampingOverride=1
    linearSpringConstant=1
    alphaBoolBlend {blendOption=Linear}
    alphaScaleBiasClamp {scale=0.4}
    boundBone {boneName=dyn_fur_back_bot_C_01}
    componentAppliedLinearAccClamp=<1000 1000 1000>
    componentLinearAccScale=<1 1 1>
    componentLinearVelScale=<1 1 1>
    physicsBodyDefinitions [{
      boundBone {boneName=dyn_fur_back_bot_C_01}
      boxExtents=<15.6 8.4 12>
      constraintSetup {angularTargetAxis=AxisZ, angularLimitsMax=<5 45 0>, angularLimitsMin=<-5 -45 -90>, angularTarget=<0 0 0>}
      localJointOffset=<1.68 0 0>
     }]
   }
   n50 = Anim Dynamics - Bone: dyn_fur_l_top_A_01(ComponentPose=@n49, Alpha=1, GravityOverride="0,0,0", ExternalForce="0,0,0") {
    angularBiasOverride=1
    angularDampingOverride=1
    angularSpringConstant=1000
    bAngularSpring=true
    bGravityOverrideInSimSpace=true
    bOverrideAngularBias=true
    bOverrideAngularDamping=true
    bOverrideLinearDamping=true
    bUseGravityOverride=true
    linearDampingOverride=1
    linearSpringConstant=1
    alphaBoolBlend {blendOption=Linear}
    alphaScaleBiasClamp {scale=0.4}
    boundBone {boneName=dyn_fur_l_top_A_01}
    componentAppliedLinearAccClamp=<1000 1000 1000>
    componentLinearAccScale=<1 1 1>
    componentLinearVelScale=<1 1 1>
    physicsBodyDefinitions [{
      boundBone {boneName=dyn_fur_l_top_A_01}
      boxExtents=<12 10.8 8.4>
      constraintSetup {angularTargetAxis=AxisZ, angularLimitsMax=<5 45 0>, angularLimitsMin=<-5 -45 -90>, angularTarget=<0 0 0>}
      localJointOffset=<0.96 0 0>
     }]
   }
   n51 = Anim Dynamics - Bone: dyn_fur_l_top_B_01(ComponentPose=@n50, Alpha=1, GravityOverride="0,0,0", ExternalForce="0,0,0") {
    angularBiasOverride=1
    angularDampingOverride=1
    angularSpringConstant=1000
    bAngularSpring=true
    bGravityOverrideInSimSpace=true
    bOverrideAngularBias=true
    bOverrideAngularDamping=true
    bOverrideLinearDamping=true
    bUseGravityOverride=true
    linearDampingOverride=1
    linearSpringConstant=1
    alphaBoolBlend {blendOption=Linear}
    alphaScaleBiasClamp {scale=0.4}
    boundBone {boneName=dyn_fur_l_top_B_01}
    componentAppliedLinearAccClamp=<1000 1000 1000>
    componentLinearAccScale=<1 1 1>
    componentLinearVelScale=<1 1 1>
    physicsBodyDefinitions [{
      boundBone {boneName=dyn_fur_l_top_B_01}
      boxExtents=<12 10.8 8.4>
      constraintSetup {angularTargetAxis=AxisZ, angularLimitsMax=<5 45 0>, angularLimitsMin=<-5 -45 -90>, angularTarget=<0 0 0>}
      localJointOffset=<0.96 0 0>
     }]
   }
   n52 = Anim Dynamics - Bone: dyn_fur_l_bot_A_01(ComponentPose=@n51, Alpha=1, GravityOverride="0,0,0", ExternalForce="0,0,0") {
    angularBiasOverride=1
    angularDampingOverride=1
    angularSpringConstant=1000
    bAngularSpring=true
    bGravityOverrideInSimSpace=true
    bOverrideAngularBias=true
    bOverrideAngularDamping=true
    bOverrideLinearDamping=true
    bUseGravityOverride=true
    linearDampingOverride=1
    linearSpringConstant=1
    alphaBoolBlend {blendOption=Linear}
    alphaScaleBiasClamp {scale=0.4}
    boundBone {boneName=dyn_fur_l_bot_A_01}
    componentAppliedLinearAccClamp=<1000 1000 1000>
    componentLinearAccScale=<1 1 1>
    componentLinearVelScale=<1 1 1>
    physicsBodyDefinitions [{
      boundBone {boneName=dyn_fur_l_bot_A_01}
      boxExtents=<12 10.8 8.4>
      constraintSetup {angularTargetAxis=AxisZ, angularLimitsMax=<5 45 0>, angularLimitsMin=<-5 -45 -90>, angularTarget=<0 0 0>}
      localJointOffset=<0.96 0 0>
     }]
   }
   n53 = Anim Dynamics - Bone: dyn_fur_l_bot_B_01(ComponentPose=@n52, Alpha=1, GravityOverride="0,0,0", ExternalForce="0,0,0") {
    angularBiasOverride=1
    angularDampingOverride=1
    angularSpringConstant=1000
    bAngularSpring=true
    bGravityOverrideInSimSpace=true
    bOverrideAngularBias=true
    bOverrideAngularDamping=true
    bOverrideLinearDamping=true
    bUseGravityOverride=true
    linearDampingOverride=1
    linearSpringConstant=1
    alphaBoolBlend {blendOption=Linear}
    alphaScaleBiasClamp {scale=0.4}
    boundBone {boneName=dyn_fur_l_bot_B_01}
    componentAppliedLinearAccClamp=<1000 1000 1000>
    componentLinearAccScale=<1 1 1>
    componentLinearVelScale=<1 1 1>
    physicsBodyDefinitions [{
      boundBone {boneName=dyn_fur_l_bot_B_01}
      boxExtents=<12 10.8 8.4>
      constraintSetup {angularTargetAxis=AxisZ, angularLimitsMax=<5 45 0>, angularLimitsMin=<-5 -45 -90>, angularTarget=<0 0 0>}
      localJointOffset=<0.96 0 0>
     }]
   }
   n54 = Anim Dynamics - Bone: dyn_fur_r_top_A_01(ComponentPose=@n53, Alpha=1, GravityOverride="0,0,0", ExternalForce="0,0,0") {
    angularBiasOverride=1
    angularDampingOverride=1
    angularSpringConstant=1000
    bAngularSpring=true
    bGravityOverrideInSimSpace=true
    bOverrideAngularBias=true
    bOverrideAngularDamping=true
    bOverrideLinearDamping=true
    bUseGravityOverride=true
    linearDampingOverride=1
    linearSpringConstant=1
    alphaBoolBlend {blendOption=Linear}
    alphaScaleBiasClamp {scale=0.4}
    boundBone {boneName=dyn_fur_r_top_A_01}
    componentAppliedLinearAccClamp=<1000 1000 1000>
    componentLinearAccScale=<1 1 1>
    componentLinearVelScale=<1 1 1>
    physicsBodyDefinitions [{
      boundBone {boneName=dyn_fur_r_top_A_01}
      boxExtents=<16.8 8.4 7.2>
      constraintSetup {angularTargetAxis=AxisZ, angularLimitsMax=<5 45 0>, angularLimitsMin=<-5 -45 -85>, angularTarget=<0 0 0>}
      localJointOffset=<1.08 0 0>
     }]
   }
   n55 = Anim Dynamics - Bone: dyn_fur_r_top_B_01(ComponentPose=@n54, Alpha=1, GravityOverride="0,0,0", ExternalForce="0,0,0") {
    angularBiasOverride=1
    angularDampingOverride=1
    angularSpringConstant=1000
    bAngularSpring=true
    bGravityOverrideInSimSpace=true
    bOverrideAngularBias=true
    bOverrideAngularDamping=true
    bOverrideLinearDamping=true
    bUseGravityOverride=true
    linearDampingOverride=1
    linearSpringConstant=1
    alphaBoolBlend {blendOption=Linear}
    alphaScaleBiasClamp {scale=0.4}
    boundBone {boneName=dyn_fur_r_top_B_01}
    componentAppliedLinearAccClamp=<1000 1000 1000>
    componentLinearAccScale=<1 1 1>
    componentLinearVelScale=<1 1 1>
    physicsBodyDefinitions [{
      boundBone {boneName=dyn_fur_r_top_B_01}
      boxExtents=<16.8 8.4 7.2>
      constraintSetup {angularTargetAxis=AxisZ, angularLimitsMax=<5 45 0>, angularLimitsMin=<-5 -45 -85>, angularTarget=<0 0 0>}
      localJointOffset=<1.08 0 0>
     }]
   }
   n56 = Anim Dynamics - Bone: dyn_fur_r_top_C_01(ComponentPose=@n55, Alpha=1, GravityOverride="0,0,0", ExternalForce="0,0,0") {
    angularBiasOverride=1
    angularDampingOverride=1
    angularSpringConstant=1000
    bAngularSpring=true
    bGravityOverrideInSimSpace=true
    bOverrideAngularBias=true
    bOverrideAngularDamping=true
    bOverrideLinearDamping=true
    bUseGravityOverride=true
    linearDampingOverride=1
    linearSpringConstant=1
    alphaBoolBlend {blendOption=Linear}
    alphaScaleBiasClamp {scale=0.4}
    boundBone {boneName=dyn_fur_r_top_C_01}
    componentAppliedLinearAccClamp=<1000 1000 1000>
    componentLinearAccScale=<1 1 1>
    componentLinearVelScale=<1 1 1>
    physicsBodyDefinitions [{
      boundBone {boneName=dyn_fur_r_top_C_01}
      boxExtents=<16.8 8.4 7.2>
      constraintSetup {angularTargetAxis=AxisZ, angularLimitsMax=<5 45 0>, angularLimitsMin=<-5 -45 -85>, angularTarget=<0 0 0>}
      localJointOffset=<1.08 0 0>
     }]
   }
   n57 = Anim Dynamics - Bone: dyn_fur_r_bot_A_01(ComponentPose=@n56, Alpha=1, GravityOverride="0,0,0", ExternalForce="0,0,0") {
    angularBiasOverride=1
    angularDampingOverride=1
    angularSpringConstant=1000
    bAngularSpring=true
    bGravityOverrideInSimSpace=true
    bOverrideAngularBias=true
    bOverrideAngularDamping=true
    bOverrideLinearDamping=true
    bUseGravityOverride=true
    linearDampingOverride=1
    linearSpringConstant=1
    alphaBoolBlend {blendOption=Linear}
    alphaScaleBiasClamp {scale=0.4}
    boundBone {boneName=dyn_fur_r_bot_A_01}
    componentAppliedLinearAccClamp=<1000 1000 1000>
    componentLinearAccScale=<1 1 1>
    componentLinearVelScale=<1 1 1>
    physicsBodyDefinitions [{
      boundBone {boneName=dyn_fur_r_bot_A_01}
      boxExtents=<16.8 8.4 7.2>
      constraintSetup {angularTargetAxis=AxisZ, angularLimitsMax=<5 45 0>, angularLimitsMin=<-5 -45 -85>, angularTarget=<0 0 0>}
      localJointOffset=<1.08 0 0>
     }]
   }
   n58 = Anim Dynamics - Bone: dyn_fur_r_bot_B_01(ComponentPose=@n57, Alpha=1, GravityOverride="0,0,0", ExternalForce="0,0,0") {
    angularBiasOverride=1
    angularDampingOverride=1
    angularSpringConstant=1000
    bAngularSpring=true
    bGravityOverrideInSimSpace=true
    bOverrideAngularBias=true
    bOverrideAngularDamping=true
    bOverrideLinearDamping=true
    bUseGravityOverride=true
    linearDampingOverride=1
    linearSpringConstant=1
    alphaBoolBlend {blendOption=Linear}
    alphaScaleBiasClamp {scale=0.4}
    boundBone {boneName=dyn_fur_r_bot_B_01}
    componentAppliedLinearAccClamp=<1000 1000 1000>
    componentLinearAccScale=<1 1 1>
    componentLinearVelScale=<1 1 1>
    physicsBodyDefinitions [{
      boundBone {boneName=dyn_fur_r_bot_B_01}
      boxExtents=<16.8 8.4 7.2>
      constraintSetup {angularTargetAxis=AxisZ, angularLimitsMax=<5 45 0>, angularLimitsMin=<-5 -45 -85>, angularTarget=<0 0 0>}
      localJointOffset=<1.08 0 0>
     }]
   }
   Output Pose(Result=Blend Poses by bool(BlendPose_0=Component To Local(ComponentPose=@n29), BlendPose_1=Component To Local(ComponentPose=@n58), bActiveValue=Get bIsInFrontEnd()) {blendType=Linear}) {layerGroup=None}
   AnimGraphNode_SubInput_0 = Input Pose()
   Save cached pose 'Input_pose'(Pose=@AnimGraphNode_SubInput_0) {cacheName=Input_pose}
  ]}
 EventGraph {nodes [
   Event BlueprintUpdateAnimation()
   > n1 = Gravity Override State Function DEPRECATED(CharacterPartAnimInstance=Self-Reference(), GameTimeInSeconds=GetGameTimeInSeconds(), GravityOverrideParams=Get Gravity Override Params())
   > Set GravityOverride(GravityOverride=@n1)
   > n2 = Branch(Condition=Get bIsInFrontEnd())
   > Set AnimDynamics_SimSpace_Frontend(AnimDynamics_SimSpace_Frontend:AnimPhysSimSpaceType Enum=World) -> (Output_Get:AnimPhysSimSpaceType Enum)
   > Set Anim_Dynamics_Alpha_Frontend(Anim_Dynamics_Alpha_Frontend:double=0.6) -> (Output_Get:double)
   Set AnimDynamics_SimSpace_Frontend(execute=@n2.else, AnimDynamics_SimSpace_Frontend=Component)
   > Set Anim_Dynamics_Alpha_Frontend(Anim_Dynamics_Alpha_Frontend=1)
   TryGetPawnOwner()
  ]}
]
variables [
 Anim_Dynamics_Alpha_Frontend {type=number}
 AnimDynamics_SimSpace_Frontend {type=AnimPhysSimSpaceType}
 Gravity Override Params {type=GravityOverrideParamsStruct}
 GravityOverride {type=Vector}
]
